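# Niagara system for the swing trail of the StrideMice bass pickaxe cosmetic, drawn as a short-lived, translucent red ribbon.
# Its single emitter uses AnimTrail modules to spawn and update ribbon particles from bone or socket indices, with a skeletal mesh exposed as a user parameter.
# Curves taper width and fade alpha over particle life, while time-of-day color and alpha curves plus a front-end/gameplay modifier adjust the look.
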NiagaraSystem /SparksCosmetics/Instrument/Bass/StrideMice/Effects/NS_Pickaxe_Bass_StrideMice_Trail
dependencies {
 usedDataInterfaces [&NiagaraDataInterfaceCurve, &NiagaraDataInterfaceColorCurve]
 usedDynamicInputs [&MakeCustomFloatFromBool, &ColorFromCurve]
 usedModules [
  &self:SystemSpawnScript.SystemScriptSource.SystemScriptGraph.NiagaraNodeAssignment_1.SetVariables_#1
  &SystemState
 ]
}
emitters {Trail_Instrument_Color {
  emitterUpdateScript [
   &EmitterState(Life Cycle Mode=::NewEnumerator1, Loop Behavior=::NewEnumerator1, Loop Duration=0.22, UseLoopDelay=true)
   &FN_SpawnAnimTrail_v2()
  ]
  particleSpawnScript [
   /Game/Effects/Niagara/Modules/AnimTrail/FN_InitializeAnimTrail/FN_InitializeAnimTrailBoneIndecies(First Bone or Socket Index=1)
   /Game/Effects/Niagara/Modules/AnimTrail/FN_InitializeAnimTrail/FN_InitializeAnimTrail(AnimTrailWidth=1.2, Color=<rgba 0.7653 0.1036 0.1042 0.6>, Lifetime=0.13)
   &SystemLocation(MaskSpawnGroup=false)
  ]
  particleUpdateScript [
   &ScaleRibbonWidth(Ribbon Width Scale=&FloatFromCurve(FloatCurve=DataInterface() {propertyValues {Curve {keys [
         {arriveTangent=-1.2579, interpMode=RCIM_Cubic, time=0, value=1}
         {arriveTangent=-2.7373, interpMode=RCIM_Cubic, leaveTangent=-2.7373, tangentMode=RCTM_User, time=1, value=0}
        ]}}}))
   &ScaleColor(
    Scale Alpha=&Multiply_Float(
     A=&FloatFromCurve(FloatCurve=DataInterface() {propertyValues {Curve {keys [
          {interpMode=RCIM_Cubic, leaveTangent=-2.3646, tangentMode=RCTM_Break, time=0, value=1}
          {arriveTangent=-0.7436, interpMode=RCIM_Cubic, leaveTangent=-0.7436, tangentMode=RCTM_User, time=0.4, value=0.1659}
          {arriveTangent=-0.2885, interpMode=RCIM_Cubic, tangentMode=RCTM_Break, time=1, value=0}
         ]}}})
     B=&FloatFromCurve(
      FloatCurve=DataInterface() {propertyValues {Curve {keys [
          {interpMode=RCIM_Cubic, leaveTangentWeight=0.1022, tangentMode=RCTM_User, time=0.0039, value=0.95}
          {arriveTangentWeight=0.1022, interpMode=RCIM_Cubic, leaveTangentWeight=0.2305, tangentMode=RCTM_User, time=0.3105, value=0.8948}
          {arriveTangent=0.0842, interpMode=RCIM_Cubic, leaveTangent=0.0842, tangentMode=RCTM_User, time=1.0018, value=0.95}
         ]}}}
      CurveIndex=Linked() {linkedVariable {name=NPC.FortniteNPC.TimeOfDay_0-1, type=NiagaraFloat}}
      Scale Curve=&FN_FrontEndFloatModifier(Gameplay=0.5, Front End=0.4)
     )
    )
    Scale RGB=&Multiply_Vector(
     A=&MakeVectorFromLinearColorRGB(Color=$Particles.Initial.Color)
     B=&MakeVectorFromLinearColorRGB(Color=$System.TOD Color)
    )
    ScaleA=true
    ScaleRGB=true
   ) {module=ScaleColor002}
   &FN_UpdateAnimTrail_v2()
  ]
  rendererClasses [&NiagaraRibbonRendererProperties]
  renderers [{rendererClass=&NiagaraRibbonRendererProperties}]
 }}
summary {
 systemName=NS_Pickaxe_Bass_StrideMice_Trail
 emitters [{emitterName=Trail_Instrument_Color, rendererClasses [&NiagaraRibbonRendererProperties]}]
 userVariables [
  User.bIsFrontEnd {type=NiagaraBool}
  User.bIsFrontEndPreview {type=NiagaraBool}
  User.RibbonIDOffset {type=NiagaraInt32}
  User.SkeletalMesh {
   type=NiagaraDataInterfaceSkeletalMesh
   defaultValue {
    struct=NiagaraExt_VariableValue_DataInterface
    value {dataInterface=&self:NiagaraDataInterfaceSkeletalMesh_0}
   }
  }
 ]
}
